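# Niagara ribbon trail effect for the RainbowHouse guitar cosmetic, built from three short-lived AnimTrail emitters that follow a pair of sockets through a shared Socket Reader.
# Hard uses the variant color, Soft a fixed muted brown, and Accent draws three trails in the accent color with per-trail widths and owner-rotated offsets.
# Each emitter fades alpha over particle life, with Soft fading more gradually, and applies time-of-day color scaling that slightly cools the tint at night.
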NiagaraSystem /SparksCosmetics/Instrument/Guitar/RainbowHouse/Effects/NS_Guitar_RainbowHouse_Trail
dependencies {
 usedDataInterfaces [
  &NiagaraDataInterfaceCurve
  &NiagaraDataInterfaceArrayFloat3
  &NiagaraDataInterfaceArrayFloat
  &NiagaraDataInterfaceColorCurve
  &NiagaraDataInterfaceSkeletalMesh
  &NiagaraDataInterfaceSocketReader
 ]
 usedDynamicInputs [&LerpLinearColors, &ColorFromCurve]
 usedModules [
  &self:SystemSpawnScript.SystemScriptSource.SystemScriptGraph.NiagaraNodeAssignment_2.SetVariables_#1
  &SystemState
  &self:SystemSpawnScript.SystemScriptSource.SystemScriptGraph.NiagaraNodeAssignment_0.SetVariables_#2
  &self:SystemSpawnScript.SystemScriptSource.SystemScriptGraph.NiagaraNodeAssignment_1.SetVariables_#3
 ]
}
emitters {
 FN_AnimTrail_Accent {
  emitterSpawnScript [&FN_RibbonIDOffsetUtility()]
  emitterUpdateScript [
   &EmitterState()
   &FN_SpawnAnimTrail(Number of Anim Trails=3)
  ]
  particleSpawnScript [
   &FN_InitializeAnimTrailBoneIndecies(First Bone or Socket Index=1, Second Bone or Socket Index=0)
   &FN_InitializeAnimTrailBoneIndecies(AnimTrail Selection=1, First Bone or Socket Index=1, Second Bone or Socket Index=0) {module=FN_InitializeAnimTrailBoneIndecies001}
   &FN_InitializeAnimTrailBoneIndecies(AnimTrail Selection=2, First Bone or Socket Index=1, Second Bone or Socket Index=0) {module=FN_InitializeAnimTrailBoneIndecies002}
   /Game/Effects/Niagara/Modules/AnimTrail/FN_InitializeAnimTrail/FN_InitializeAnimTrail_SocketReader/FN_InitializeAnimTrail_SocketReader(
    AnimTrailWidth:NiagaraFloat=&SelectFloatFromArray(
     Direct Array Index=&SpawnGroup()
     Float Selection Array=DataInterface() {propertyValues {FloatData [0.3,0.28,0.1]}}
    )
    Color:LinearColor=$User.VariantColorAccent
    Lifetime=0.2
    Position Offset:Vector3f=&RotateVectorByQuaternion(
     Vector To Rotate=&SelectVectorFromArray(
      Direct Array Index=&SpawnGroup()
      Vector Selection Array=DataInterface() {propertyValues {FloatData [<5 0 8.5>, <0 0 -12>, <0 0 -2>]}}
     )
     Quaternion=$Engine.Owner.Rotation
    )
    Socket Reader:NiagaraDataInterfaceSocketReader=$System.Socket Reader
   )
   &AddVelocity(Velocity=<800 0 0>) {enabled=false}
  ]
  particleUpdateScript [
   &SolveForcesAndVelocity(Clamp Velocity=false, Limit Acceleration=false)
   /Game/Effects/Niagara/Modules/AnimTrail/FN_UpdateAnimTrail/FN_UpdateAnimTrail_SocketReader/FN_UpdateAnimTrail_SocketReader(Socket Reader:NiagaraDataInterfaceSocketReader=$System.Socket Reader)
   &ScaleColor(
    Scale Alpha=&Multiply_Float(
     A=&FloatFromCurve(
      FloatCurve=DataInterface() {propertyValues {Curve {keys [
          {interpMode=RCIM_Cubic, leaveTangentWeight=0.259, tangentMode=RCTM_Break, time=0, value=1}
          {arriveTangent=-1.5891, interpMode=RCIM_Cubic, leaveTangent=-1.5891, tangentMode=RCTM_User, time=0.2, value=0.25}
          {arriveTangent=-0.2139, interpMode=RCIM_Cubic, tangentMode=RCTM_Break, time=0.5, value=0}
         ]}}}
      Scale Curve=0.5
     )
     B=&FNTimeOfDayScalar(Night Scale=1, Morning Scale=1, Evening Scale=1)
    )
    Scale RGB=&FN_Select_Vector_from_Time_Of_Day(Morning=<1 1 1>, Day=<1 1 1>, Evening=<1 1 1>, Night=<0.8 0.8 1>)
    ScaleA=true
    ScaleRGB=true
   ) {module=ScaleColor002}
  ]
  rendererClasses [&NiagaraRibbonRendererProperties]
  renderers [{rendererClass=&NiagaraRibbonRendererProperties}]
 }
 FN_AnimTrail_Hard {
  emitterSpawnScript [&FN_RibbonIDOffsetUtility()]
  emitterUpdateScript [
   &EmitterState()
   &FN_SpawnAnimTrail()
  ]
  particleSpawnScript [
   &FN_InitializeAnimTrailBoneIndecies(First Bone or Socket Index=1, Second Bone or Socket Index=0)
   &FN_InitializeAnimTrail_SocketReader(Color:LinearColor=$User.VariantColor, Lifetime=0.2, Socket Reader:NiagaraDataInterfaceSocketReader=$System.Socket Reader)
   &AddVelocity(Velocity=<800 0 0>) {enabled=false}
  ]
  particleUpdateScript [
   &SolveForcesAndVelocity(Clamp Velocity=false, Limit Acceleration=false)
   &FN_UpdateAnimTrail_SocketReader(Socket Reader:NiagaraDataInterfaceSocketReader=$System.Socket Reader)
   &ScaleColor(
    Scale Alpha=&Multiply_Float(
     A=&FloatFromCurve(
      FloatCurve=DataInterface() {propertyValues {Curve {keys [
          {interpMode=RCIM_Cubic, leaveTangentWeight=0.259, tangentMode=RCTM_Break, time=0, value=1}
          {arriveTangent=-1.5891, interpMode=RCIM_Cubic, leaveTangent=-1.5891, tangentMode=RCTM_User, time=0.2, value=0.25}
          {arriveTangent=-0.2139, interpMode=RCIM_Cubic, tangentMode=RCTM_Break, time=0.5, value=0}
         ]}}}
      Scale Curve=0.5
     )
     B=&FNTimeOfDayScalar(Night Scale=1, Morning Scale=1, Evening Scale=1)
    )
    Scale RGB=&FN_Select_Vector_from_Time_Of_Day(Morning=<1 1 1>, Day=<1 1 1>, Evening=<1 1 1>, Night=<0.8 0.8 1>)
    ScaleA=true
    ScaleRGB=true
   ) {module=ScaleColor002}
  ]
  rendererClasses [&NiagaraRibbonRendererProperties]
  renderers [{rendererClass=&NiagaraRibbonRendererProperties}]
 }
 FN_AnimTrail_Soft {
  emitterSpawnScript [&FN_RibbonIDOffsetUtility()]
  emitterUpdateScript [
   &EmitterState()
   &FN_SpawnAnimTrail()
  ]
  particleSpawnScript [
   &FN_InitializeAnimTrailBoneIndecies(First Bone or Socket Index=1, Second Bone or Socket Index=0)
   &FN_InitializeAnimTrail_SocketReader(Color:LinearColor=<rgba 0.1774 0.127 0.104 1>, Lifetime=0.2, Socket Reader:NiagaraDataInterfaceSocketReader=$System.Socket Reader)
   &AddVelocity(Velocity=<800 0 0>) {enabled=false}
  ]
  particleUpdateScript [
   &SolveForcesAndVelocity(Clamp Velocity=false, Limit Acceleration=false)
   &FN_UpdateAnimTrail_SocketReader(Socket Reader:NiagaraDataInterfaceSocketReader=$System.Socket Reader)
   &ScaleColor(
    Scale Alpha=&Multiply_Float(
     A=&FloatFromCurve(
      FloatCurve=DataInterface() {propertyValues {Curve {keys [
          {interpMode=RCIM_Cubic, leaveTangentWeight=0.259, tangentMode=RCTM_Break, time=0, value=1}
          {arriveTangent=-1, interpMode=RCIM_Cubic, leaveTangent=-1, time=0.6689, value=0.5}
          {arriveTangent=-1.562, interpMode=RCIM_Cubic, tangentMode=RCTM_Break, time=1, value=0}
         ]}}}
      Scale Curve=0.5
     )
     B=&FNTimeOfDayScalar(Night Scale=1, Morning Scale=1, Evening Scale=1)
    )
    Scale RGB=&FN_Select_Vector_from_Time_Of_Day(Morning=<1 1 1>, Day=<1 1 1>, Evening=<1 1 1>, Night=<0.8 0.8 1>)
    ScaleA=true
    ScaleRGB=true
   ) {module=ScaleColor002}
  ]
  rendererClasses [&NiagaraRibbonRendererProperties]
  renderers [{rendererClass=&NiagaraRibbonRendererProperties}]
 }
}
summary {
 systemName=NS_Guitar_RainbowHouse_Trail
 emitters [
  {emitterName=FN_AnimTrail_Hard, rendererClasses [&NiagaraRibbonRendererProperties]}
  {emitterName=FN_AnimTrail_Soft, rendererClasses [&NiagaraRibbonRendererProperties]}
  {emitterName=FN_AnimTrail_Accent, rendererClasses [&NiagaraRibbonRendererProperties]}
 ]
 userVariables [
  User.RibbonIDOffset {type=NiagaraInt32}
  User.VariantColor {type=LinearColor, description="", defaultValue {struct=LinearColor, value=<rgba 1 0 0 1>}}
  User.VariantColorAccent {type=LinearColor, description="", defaultValue {struct=LinearColor, value=<rgba 0.7096 0.4862 0.1123 1>}}
 ]
}
